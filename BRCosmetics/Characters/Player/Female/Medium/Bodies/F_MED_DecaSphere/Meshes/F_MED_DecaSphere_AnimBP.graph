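# Animation Blueprint for the F_MED_DecaSphere female player body cosmetic, following the parent mesh's pose and adding secondary motion to its laces.
# Copy Pose From Mesh feeds the DecaSphere Control Rig, then seven chained Anim Dynamics nodes simulate left and right lace and lace-rope bones.
# The simulations are root-relative, with damped angular limits and a spherical limit on the left calf for lace_l_2.
# The event graph has no logic.
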AnimBlueprint /BRCosmetics/Characters/Player/Female/Medium/Bodies/F_MED_DecaSphere/Meshes/F_MED_DecaSphere_AnimBP
target_skeleton=&/F_MED_DecaSphere_Skeleton
graphs [
 AnimGraph {nodes [
   n1 = Control Rig(Source=Copy Pose From Mesh() {bUseAttachedParent=true}) {
    alphaBoolBlend {blendOption=Linear}
    controlRigAssetReference {blueprintRigClass=&/F_MED_DecaSphere_CtrlRig.F_MED_DecaSphere_CtrlRig_C}
   }
   n2 = Anim Dynamics - Bone: lace_l_1(ComponentPose=Local To Component(LocalPose=@n1), Alpha=1, ExternalForce="0,0,0") {
    angularDampingOverride=0.999
    angularSpringConstant=100
    bOverrideAngularDamping=true
    simulationSpace=RootRelative
    alphaBoolBlend {blendOption=Linear}
    boundBone {boneName=lace_l_1}
    componentAppliedLinearAccClamp=<1000 1000 1000>
    componentLinearAccScale=<0.3 0.3 0.3>
    componentLinearVelScale=<0.3 0.3 0.3>
    gravityOverride {z=-980}
    physicsBodyDefinitions [{
      boundBone {boneName=lace_l_1}
      boxExtents=<2 2 2>
      constraintSetup {angularLimitsMax=<2 30 30>, angularLimitsMin=<-2 -30 0>, angularTarget=<1 0 0>}
      localJointOffset=<0.05 0 0>
     }]
    relativeSpaceBone {boneName=calf_r}
   }
   n3 = Anim Dynamics - Bone: lace_l_2(ComponentPose=@n2, Alpha=1, ExternalForce="0,0,0") {
    angularDampingOverride=0.999
    angularSpringConstant=100
    bOverrideAngularDamping=true
    simulationSpace=RootRelative
    alphaBoolBlend {blendOption=Linear}
    boundBone {boneName=lace_l_2}
    componentAppliedLinearAccClamp=<1000 1000 1000>
    componentLinearAccScale=<0.3 0.3 0.3>
    componentLinearVelScale=<0.3 0.3 0.3>
    gravityOverride {z=-980}
    physicsBodyDefinitions [{
      boundBone {boneName=lace_l_2}
      boxExtents=<1.5 1.5 1.5>
      constraintSetup {angularLimitsMax=<5 15 45>, angularLimitsMin=<-5 -15 0>, angularTarget=<1 0 0>}
      localJointOffset=<0.08 0 0>
     }]
    relativeSpaceBone {boneName=lace_l_1}
    sphericalLimits [{limitRadius=8, drivingBone {boneName=calf_l}, sphereLocalOffset=<-24.2755 -5.2506 -0.6934>}]
   }
   n4 = Anim Dynamics - Bone: lace_r_1(ComponentPose=@n3, Alpha=1, ExternalForce="0,0,0") {
    angularDampingOverride=0.999
    angularSpringConstant=100
    bOverrideAngularDamping=true
    simulationSpace=RootRelative
    alphaBoolBlend {blendOption=Linear}
    boundBone {boneName=lace_r_1}
    componentAppliedLinearAccClamp=<1000 1000 1000>
    componentLinearAccScale=<0.1 0.1 0.1>
    componentLinearVelScale=<0.1 0.1 0.1>
    gravityOverride {z=-980}
    physicsBodyDefinitions [{
      boundBone {boneName=lace_r_1}
      boxExtents=<2 2 2>
      constraintSetup {angularLimitsMax=<10 20 45>, angularLimitsMin=<-10 -20 -10>, angularTarget=<1 0 0>}
      localJointOffset=<0.04 0 0>
     }]
    relativeSpaceBone {boneName=calf_twist_02_r}
   }
   n5 = Anim Dynamics - Bone: lace_rope_r_1(ComponentPose=@n4, Alpha=1, ExternalForce="0,0,0") {
    angularDampingOverride=0.999
    angularSpringConstant=100
    bOverrideAngularDamping=true
    simulationSpace=RootRelative
    alphaBoolBlend {blendOption=Linear}
    boundBone {boneName=lace_rope_r_1}
    componentAppliedLinearAccClamp=<1000 1000 1000>
    componentLinearAccScale=<0.1 0.1 0.1>
    componentLinearVelScale=<0.1 0.1 0.1>
    gravityOverride {z=-980}
    physicsBodyDefinitions [{
      boundBone {boneName=lace_rope_r_1}
      boxExtents=<2 2 2>
      constraintSetup {angularLimitsMax=<10 20 45>, angularLimitsMin=<-10 -20 -10>, angularTarget=<1 0 0>}
      localJointOffset=<0.04 0 0>
     }]
    relativeSpaceBone {boneName=calf_twist_02_r}
   }
   n6 = Anim Dynamics - Bone: lace_rope_r_2(ComponentPose=@n5, Alpha=1, ExternalForce="0,0,0") {
    angularDampingOverride=0.999
    angularSpringConstant=100
    bOverrideAngularDamping=true
    simulationSpace=RootRelative
    alphaBoolBlend {blendOption=Linear}
    boundBone {boneName=lace_rope_r_2}
    componentAppliedLinearAccClamp=<1000 1000 1000>
    componentLinearAccScale=<0.1 0.1 0.1>
    componentLinearVelScale=<0.1 0.1 0.1>
    gravityOverride {z=-980}
    physicsBodyDefinitions [{
      boundBone {boneName=lace_rope_r_2}
      boxExtents=<2 2 2>
      constraintSetup {angularLimitsMax=<10 20 45>, angularLimitsMin=<-10 -20 -10>, angularTarget=<1 0 0>}
      localJointOffset=<0.04 0 0>
     }]
    relativeSpaceBone {boneName=calf_twist_02_r}
   }
   n7 = Anim Dynamics - Bone: lace_rope_l_1(ComponentPose=@n6, Alpha=1, ExternalForce="0,0,0") {
    angularDampingOverride=0.999
    angularSpringConstant=100
    bOverrideAngularDamping=true
    simulationSpace=RootRelative
    alphaBoolBlend {blendOption=Linear}
    boundBone {boneName=lace_rope_l_1}
    componentAppliedLinearAccClamp=<1000 1000 1000>
    componentLinearAccScale=<0.1 0.1 0.1>
    componentLinearVelScale=<0.1 0.1 0.1>
    gravityOverride {z=-980}
    physicsBodyDefinitions [{
      boundBone {boneName=lace_rope_l_1}
      boxExtents=<2 2 2>
      constraintSetup {angularLimitsMax=<10 20 45>, angularLimitsMin=<-10 -20 -10>, angularTarget=<1 0 0>}
      localJointOffset=<0.04 0 0>
     }]
    relativeSpaceBone {boneName=calf_twist_02_r}
   }
   n8 = Anim Dynamics - Bone: lace_rope_l_2(ComponentPose=@n7, Alpha=1, ExternalForce="0,0,0") {
    angularDampingOverride=0.999
    angularSpringConstant=100
    bOverrideAngularDamping=true
    simulationSpace=RootRelative
    alphaBoolBlend {blendOption=Linear}
    boundBone {boneName=lace_rope_l_2}
    componentAppliedLinearAccClamp=<1000 1000 1000>
    componentLinearAccScale=<0.1 0.1 0.1>
    componentLinearVelScale=<0.1 0.1 0.1>
    gravityOverride {z=-980}
    physicsBodyDefinitions [{
      boundBone {boneName=lace_rope_l_2}
      boxExtents=<2 2 2>
      constraintSetup {angularLimitsMax=<10 20 45>, angularLimitsMin=<-10 -20 -10>, angularTarget=<1 0 0>}
      localJointOffset=<0.04 0 0>
     }]
    relativeSpaceBone {boneName=calf_twist_02_r}
   }
   Output Pose(Result=Component To Local(ComponentPose=@n8)) {layerGroup=None}
  ]}
 EventGraph {nodes [Event BlueprintUpdateAnimation()]}
]
variables []
