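# Physics asset for the F_MED_DecaSphere female face accessory mesh, setting up a five-joint dynamic hair chain with upper-body collision shapes.
# Capsule bodies cover the hair joints plus head, neck, spine, clavicles and both arms.
# Angular-limited constraints link the hair joints in sequence from the head, with a tight twist throughout and an asymmetric swing at the root.
# The other body capsules have no constraints between them.
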PhysicsAsset /BRCosmetics/Characters/Player/Female/Medium/Bodies/F_MED_DecaSphere/Meshes/Parts/F_MED_DecaSphere_FaceAcc_Physics
bodies {
 clavicle_l [{
   shapeType=Capsule
   length=14.5188
   radius=8.9174
   shapeName=clavicle_l_capsule
   center=<6.1917 0.0126 0.1414>
   rotation=<rot -80.2051 179.9982 -179.9981>
  }]
 clavicle_r [{
   shapeType=Capsule
   length=13.3561
   radius=9.0553
   shapeName=clavicle_r_capsule
   center=<-6.3783 -0.1593 -0.046>
   rotation=<rot -85.2003 179.9972 -179.9972>
  }]
 dyn_hair_joint_1 [{
   shapeType=Capsule
   length=9.0183
   radius=6.6141
   shapeName=dyn_hair_joint_1_capsule
   center=<6.8973 -1.8962 0.4181>
   rotation=<rot 90 180 180>
  }]
 dyn_hair_joint_2 [{
   shapeType=Capsule
   length=3.4309
   radius=5.8293
   shapeName=dyn_hair_joint_2_capsule
   center=<4.2135 -1.6163 0.9693>
   rotation=<rot 88.344 -179.9888 -180>
  }]
 dyn_hair_joint_3 [{
   shapeType=Capsule
   length=6.1372
   radius=5.5302
   shapeName=dyn_hair_joint_3_capsule
   center=<3.5104 -1.2508 0.4445>
   rotation=<rot 90 180 180>
  }]
 dyn_hair_joint_4 [{shapeType=Capsule, length=7.9487, radius=4.7014, shapeName=dyn_hair_joint_4_capsule, center=<6.0008 -1.294 1.1382>, rotation=<rot 90 0 0>}]
 dyn_hair_joint_5 [{
   shapeType=Capsule
   length=6.5084
   radius=4.6691
   shapeName=dyn_hair_joint_5_capsule
   center=<5.6597 -0.7615 1.0826>
   rotation=<rot 88.1659 -179.7019 177.8977>
  }]
 head [{shapeType=Capsule, length=8.5584, radius=10.5438, center=<-2.2308 -3.1854 -1.176>, rotation=<rot 3.3286 93.2053 1.7684>}]
 lowerarm_l [{
   shapeType=Capsule
   length=25.9907
   radius=7.4346
   shapeName=lowerarm_l_capsule
   center=<21.9138 0.0001 -1.8767>
   rotation=<rot 82.2344 -0.1121 -0.1298>
  }]
 lowerarm_r [{
   shapeType=Capsule
   length=25.9907
   radius=7.4346
   shapeName=lowerarm_r_capsule
   center=<-20.0231 -0.0001 1.7147>
   rotation=<rot 81.0657 -0.1145 -0.1361>
  }]
 neck_01 [{
   shapeType=Capsule
   length=13.998
   radius=7.4584
   shapeName=neck_01_capsule
   center=<-0.2743 -0.0693 0.0062>
   rotation=<rot 0.7063 -14.7818 0.3325>
  }]
 spine_01 [{shapeType=Capsule, length=19.905, radius=11.715, shapeName=spine_01_capsule}]
 spine_02 [{shapeType=Capsule, length=19.905, radius=11.715, shapeName=spine_02_capsule}]
 spine_03 [{shapeType=Capsule, length=19.905, radius=11.715, shapeName=spine_03_capsule, center=<5.3096 0.4826 0>}]
 spine_04 [{shapeType=Capsule, length=19.905, radius=11.715, shapeName=spine_04_capsule, center=<6.0434 0.5137 0>}]
 spine_05 [{shapeType=Capsule, length=19.905, radius=11.715, shapeName=spine_05_capsule, center=<2.8423 1.5614 0>}]
 upperarm_l [{
   shapeType=Capsule
   length=24.3194
   radius=7.1544
   shapeName=upperarm_l_capsule
   center=<15.7238 0 0.974>
   rotation=<rot -87.194 0.0042 -0.004>
  }]
 upperarm_r [{
   shapeType=Capsule
   length=24.3194
   radius=7.1544
   shapeName=upperarm_r_capsule
   center=<-14.3111 0 -0.2671>
   rotation=<rot -88.9306 0.0107 -0.0105>
  }]
}
constraints [
 {bone1Name=dyn_hair_joint_1, bone2Name=head, swing1LimitDegrees=60, swing2LimitDegrees=15, twistLimitDegrees=2}
 {bone1Name=dyn_hair_joint_2, bone2Name=dyn_hair_joint_1, swing1LimitDegrees=30, swing2LimitDegrees=30, twistLimitDegrees=5}
 {bone1Name=dyn_hair_joint_3, bone2Name=dyn_hair_joint_2, swing1LimitDegrees=30, swing2LimitDegrees=30, twistLimitDegrees=5}
 {bone1Name=dyn_hair_joint_4, bone2Name=dyn_hair_joint_3, swing1LimitDegrees=25, swing2LimitDegrees=25, twistLimitDegrees=5}
 {bone1Name=dyn_hair_joint_5, bone2Name=dyn_hair_joint_4, swing1LimitDegrees=30, swing2LimitDegrees=30, twistLimitDegrees=5}
]
