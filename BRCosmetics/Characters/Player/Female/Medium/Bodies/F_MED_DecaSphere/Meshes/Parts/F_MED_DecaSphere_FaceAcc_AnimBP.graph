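# Anim Blueprint for the DecaSphere female face accessory, simulating its dyn_hair joints with head-space RigidBody physics.
# It copies the parent pose, layering arms from the body mesh.
# With backbling, a lower-scale simulation gets bone offsets and spine_05/joint constraints, partly weighted by head rotation; otherwise a livelier simulation plays.
# Updates compute head alpha plus movement-state RBAN and gravity settings via a deprecated helper; customization load sets body and backpack.
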AnimBlueprint /BRCosmetics/Characters/Player/Female/Medium/Bodies/F_MED_DecaSphere/Meshes/Parts/F_MED_DecaSphere_FaceAcc_AnimBP
target_skeleton=&/F_MED_DecaSphere_FaceAcc_Skeleton
graphs [
 AnimGraph {nodes [
   n1 = RigidBody!(
    ComponentPose=Local To Component(LocalPose=Use cached pose 'start'())
    bAlphaBoolEnabled:bool
    OverrideWorldGravity:Vector="0,0,0"
    ExternalForce
    ComponentLinearAccScale:Vector="0.2,0.5,0.1"
    ComponentLinearVelScale:Vector="0.2,0.5,0.1"
    SimSpaceSettings:Sim Space Settings Structure=(WorldAlpha=1,VelocityScaleZ=0.7,DampingAlpha=1,MaxLinearVelocity=100,MaxAngularVelocity=0.1,MaxLinearAcceleration=2000,MaxAngularAcceleration=1,ExternalLinearDragV=(X=0,Y=0,Z=0),ExternalLinearVelocity=(X=0,Y=0,Z=0),ExternalAngularVelocity=(X=0,Y=0,Z=0))
   ) -> (Pose) {
    alphaInputType=Bool
    bOverrideWorldGravity=true
    overridePhysicsAsset=&/F_MED_DecaSphere_FaceAcc_Physics
    simulationSpace=BaseBoneSpace
    alphaBoolBlend {blendOption=Linear}
    baseBoneRef {boneName=head}
    componentAppliedLinearAccClamp=<1000 1000 1000>
    componentLinearAccScale=<0.2 0.5 0.1>
    componentLinearVelScale=<0.2 0.5 0.1>
    simSpaceSettings {maxAngularAcceleration=1, maxAngularVelocity=0.1, maxLinearAcceleration=2000, maxLinearVelocity=100, velocityScaleZ=0.7, worldAlpha=1}
   }
   n2 = RigidBody!(ComponentPose=Local To Component(LocalPose=Use cached pose 'start'()), bAlphaBoolEnabled, ExternalForce) -> (Pose) {
    alphaInputType=Bool
    overridePhysicsAsset=&F_MED_DecaSphere_FaceAcc_Physics
    simulationSpace=BaseBoneSpace
    alphaBoolBlend {blendOption=Linear}
    baseBoneRef {boneName=head}
    componentAppliedLinearAccClamp=<1000 1000 1000>
    componentLinearAccScale=<0.1 0.1 0.1>
    componentLinearVelScale=<0.1 0.1 0.1>
    simSpaceSettings {maxAngularAcceleration=0, maxAngularVelocity=0, maxLinearAcceleration=0, maxLinearVelocity=0, velocityScaleZ=0}
   }
   n3 = Get headrot() -> (headrot:float)
   n4 = "Transform (Modify) Bone - Bone: dyn_hair_joint_1"(ComponentPose=@n2, Alpha=@n3, Translation="0,0,2", Rotation="0,0,0", Scale="1,1,1") {
    rotationMode=BMM_Additive
    scaleMode=BMM_Additive
    translationMode=BMM_Additive
    alphaBoolBlend {blendOption=Linear}
    boneToModify {boneName=dyn_hair_joint_1}
    translation {z=2}
   }
   n5 = "Transform (Modify) Bone - Bone: dyn_hair_joint_3"(ComponentPose=@n4, Alpha=@n3, Translation="0,-8,0", Rotation="0,0,0", Scale="1,0.3,1") {
    rotationMode=BMM_Additive
    scaleMode=BMM_Additive
    translationMode=BMM_Additive
    alphaBoolBlend {blendOption=Linear}
    boneToModify {boneName=dyn_hair_joint_3}
    scale {y=0.3}
    translation {y=-8}
   }
   n6 = "Transform (Modify) Bone - Bone: dyn_hair_joint_2"(ComponentPose=@n5, Alpha=1, Translation="0,6,0", Rotation="0,0,0", Scale="1,0.75,1") {
    rotationMode=BMM_Additive
    scaleMode=BMM_Additive
    translationMode=BMM_Additive
    alphaBoolBlend {blendOption=Linear}
    boneToModify {boneName=dyn_hair_joint_2}
    scale {y=0.75}
    translation {y=6}
   }
   n7 = "Constraint - dyn_hair_joint_3 (P)"(ComponentPose=@n6, Alpha=0.6, ConstraintWeights_0=1) {
    alphaBoolBlend {blendOption=Linear}
    boneToModify {boneName=dyn_hair_joint_3}
    constraintSetup [{targetBone {boneName=spine_05}}]
    constraintWeights [1]
   }
   n8 = "Constraint - dyn_hair_joint_4 (P)"(ComponentPose=@n7, Alpha=1, ConstraintWeights_0=1) {
    alphaBoolBlend {blendOption=Linear}
    boneToModify {boneName=dyn_hair_joint_4}
    constraintSetup [{targetBone {boneName=spine_05}}]
    constraintWeights [1]
   }
   n9 = "Constraint - dyn_hair_joint_5 (P)"(ComponentPose=@n8, Alpha=1, ConstraintWeights_0=1) {
    alphaBoolBlend {blendOption=Linear}
    boneToModify {boneName=dyn_hair_joint_5}
    constraintSetup [{targetBone {boneName=dyn_hair_joint_4}}]
    constraintWeights [1]
   }
   n10 = "Transform (Modify) Bone - Bone: dyn_hair_joint_4"(ComponentPose=@n9, Alpha=1, Translation="0,3,0", Rotation="0,0,-10", Scale="1,1,1") {
    rotationMode=BMM_Additive
    scaleMode=BMM_Additive
    translationMode=BMM_Additive
    translationSpace=BCS_BoneSpace
    alphaBoolBlend {blendOption=Linear}
    boneToModify {boneName=dyn_hair_joint_4}
    rotation {roll=-10}
    translation {y=3}
   }
   n11 = "Transform (Modify) Bone - Bone: dyn_hair_joint_5"(ComponentPose=@n10, Alpha=1, Translation="0,4,0", Rotation="0,0,15", Scale="1,1,1") {
    rotationMode=BMM_Additive
    scaleMode=BMM_Additive
    translationMode=BMM_Additive
    translationSpace=BCS_BoneSpace
    alphaBoolBlend {blendOption=Linear}
    boneToModify {boneName=dyn_hair_joint_5}
    rotation {roll=15}
    translation {y=4}
   }
   Output Pose(Result=Blend Poses by bool(
     BlendPose_0=Component To Local(ComponentPose=@n11)
     BlendPose_1=Component To Local(ComponentPose=@n1)
     bActiveValue=Get backpack() -> (backpack:bool)
    ) {blendType=Linear}) {layerGroup=None}
   n12 = Layered blend per bone(
    BasePose=Copy Pose From Mesh() {bUseAttachedParent=true}
    BlendPoses_0=Copy Pose From Mesh(SourceMeshComponent=Get body() -> (body:Skeletal Mesh Component Budgeted Object Reference)) {bUseAttachedParent=true}
   ) {layerSetup [{branchFilters [{boneName=upperarm_l}, {boneName=upperarm_r}, {boneName=lowerarm_l}, {boneName=lowerarm_r}]}]}
   Save cached pose 'start'(Pose=@n12) {cacheName=start}
  ]}
 EventGraph {nodes [
   Event BlueprintUpdateAnimation()
   > Set headrot(headrot:float=GetAlphaFromReferenceTransformDelta(SkeletalMeshComponent=GetPartSkeletalMeshComponent(), BoneName=head, BaseName=spine_05, InputRange=(X=-10,Y=-45), TwistAxis="1,0,0")) -> (Output_Get:float)
   > n1 = Gravity Override State Function DEPRECATED(
    CharacterPartAnimInstance=Self-Reference()
    GameTimeInSeconds=GetGameTimeInSeconds()
    GravityOverrideParams=Get Gravity Override Params()
    RBAN Params=Get RBAN Params()
   )
   > Set RBAN Settings Out(RBAN Settings Out:RBANSetting Struct Structure=@n1.RBAN SettingsOut) -> (Output_Get:RBANSetting Struct Structure)
   > Set Gravity Override(Gravity Override=@n1)
   n2 = Event OnCharacterCustomizationLoaded()
   > Set body(body:Skeletal Mesh Component Budgeted Object Reference=Get Array Item with EFort Custom Part Type(Array=@n2.CustomMeshParts)) -> (Output_Get:Skeletal Mesh Component Budgeted Object Reference)
   > Set backpack(backpack:bool=CheckForBackbling()) -> (Output_Get:bool)
  ]}
]
variables [
 backpack {type=boolean}
 body {type=SkeletalMeshComponentBudgeted}
 Gravity Override {type=Vector}
 Gravity Override Params {type=GravityOverrideParamsStruct, default {tactical_Sprint {gravityOverride=<0 -450 -980>}}}
 headrot {type=number}
 RBAN Params {
  type=RBANParamsStruct
  default {
   dBNO {simSpaceSettings {
     maxAngularAcceleration=1000
     maxAngularVelocity=3
     maxLinearAcceleration=300
     maxLinearVelocity=700
     velocityScaleZ=0
     externalLinearDragV=<0 0.5 0>
    }}
   dBNO_Carried {simSpaceSettings {
     maxAngularAcceleration=1000
     maxAngularVelocity=3
     maxLinearAcceleration=300
     maxLinearVelocity=700
     velocityScaleZ=0
     externalLinearDragV=<0 0.5 0>
    }}
   emoteOrMelee {simSpaceSettings {
     maxAngularAcceleration=1000
     maxAngularVelocity=3
     maxLinearAcceleration=300
     maxLinearVelocity=700
     velocityScaleZ=0
     externalLinearDragV=<0 0.5 0>
    }}
   front_End {simSpaceSettings {
     maxAngularAcceleration=1000
     maxAngularVelocity=3
     maxLinearAcceleration=300
     maxLinearVelocity=700
     velocityScaleZ=0
     worldAlpha=0.25
     externalLinearDragV=<0 0.5 0>
    }}
   onGround_Crouch_Moving {simSpaceSettings {
     maxAngularAcceleration=1000
     maxAngularVelocity=3
     maxLinearAcceleration=300
     maxLinearVelocity=700
     velocityScaleZ=0
     externalLinearDragV=<0 0.5 0>
    }}
   onGround_Falling {simSpaceSettings {
     maxAngularAcceleration=1000
     maxAngularVelocity=3
     maxLinearAcceleration=300
     maxLinearVelocity=700
     velocityScaleZ=0
     externalLinearDragV=<0 0.5 0>
    }}
   onGround_Moving {
    componentLinearAccScale=<0.2 0.3 0.1>
    componentLinearVelScale=<0.2 0.3 0.1>
    simSpaceSettings {
     maxAngularAcceleration=1
     maxAngularVelocity=0.1
     maxLinearAcceleration=350
     maxLinearVelocity=200
     velocityScaleZ=0.5
     externalLinearDragV=<0 0 0>
    }
   }
   onGround_Standing {
    componentLinearAccScale=<0.2 1 0.2>
    componentLinearVelScale=<0.2 1 0.2>
    simSpaceSettings {
     maxAngularAcceleration=1
     maxAngularVelocity=0.1
     maxLinearAcceleration=1000
     maxLinearVelocity=100
     velocityScaleZ=1
     worldAlpha=0
     externalLinearDragV=<0 0 0>
    }
   }
   rail_Grinding {simSpaceSettings {
     maxAngularAcceleration=1000
     maxAngularVelocity=3
     maxLinearAcceleration=300
     maxLinearVelocity=700
     velocityScaleZ=0
     worldAlpha=0.1
     externalLinearDragV=<0 0.5 0>
    }}
   riding_Moving {simSpaceSettings {
     maxAngularAcceleration=1000
     maxAngularVelocity=3
     maxLinearAcceleration=300
     maxLinearVelocity=700
     velocityScaleZ=0
     externalLinearDragV=<0 0.5 0>
    }}
   riding_Moving_Targeting {simSpaceSettings {
     maxAngularAcceleration=1000
     maxAngularVelocity=3
     maxLinearAcceleration=300
     maxLinearVelocity=700
     velocityScaleZ=0
     externalLinearDragV=<0 0.5 0>
    }}
   riding_Standing {simSpaceSettings {
     maxAngularAcceleration=1000
     maxAngularVelocity=3
     maxLinearAcceleration=300
     maxLinearVelocity=700
     velocityScaleZ=0
     externalLinearDragV=<0 0.5 0>
    }}
   skydive_DivingDown {simSpaceSettings {
     maxAngularAcceleration=1000
     maxAngularVelocity=3
     maxLinearAcceleration=300
     maxLinearVelocity=700
     velocityScaleZ=0
     externalLinearDragV=<0 0.5 0>
    }}
   skydive_FloatingDown {
    componentLinearAccScale=<0.2 1 0.2>
    componentLinearVelScale=<0.2 1 0.2>
    simSpaceSettings {
     maxAngularAcceleration=1000
     maxAngularVelocity=1
     maxLinearAcceleration=1000
     maxLinearVelocity=700
     velocityScaleZ=1
     externalLinearDragV=<0 0 0>
    }
   }
   skydive_GoingUp {simSpaceSettings {
     maxAngularAcceleration=1000
     maxAngularVelocity=3
     maxLinearAcceleration=300
     maxLinearVelocity=700
     velocityScaleZ=0
     externalLinearDragV=<0 0.5 0>
    }}
   skydive_Parachuting {
    componentLinearAccScale=<0 0.1 0.3>
    componentLinearVelScale=<0 0.1 0.3>
    simSpaceSettings {
     maxAngularAcceleration=1000
     maxAngularVelocity=3
     maxLinearAcceleration=300
     maxLinearVelocity=700
     velocityScaleZ=0
     externalLinearDragV=<0 0.5 0>
    }
   }
   skydive_Surfing {simSpaceSettings {
     maxAngularAcceleration=1000
     maxAngularVelocity=3
     maxLinearAcceleration=300
     maxLinearVelocity=700
     velocityScaleZ=0
     externalLinearDragV=<0 0.5 0>
    }}
   swimming {simSpaceSettings {
     maxAngularAcceleration=1000
     maxAngularVelocity=3
     maxLinearAcceleration=300
     maxLinearVelocity=700
     velocityScaleZ=0
     externalLinearDragV=<0 0.5 0>
    }}
   tactical_Sprint {
    componentLinearAccScale=<0.2 0.3 0.1>
    componentLinearVelScale=<0.2 0.3 0.1>
    simSpaceSettings {
     maxAngularAcceleration=1
     maxAngularVelocity=0.1
     maxLinearAcceleration=350
     maxLinearVelocity=200
     velocityScaleZ=0.5
     externalLinearDragV=<0 0 0>
    }
   }
   ziplining {simSpaceSettings {
     maxAngularAcceleration=1000
     maxAngularVelocity=3
     maxLinearAcceleration=300
     maxLinearVelocity=700
     velocityScaleZ=0
     externalLinearDragV=<0 0.5 0>
    }}
  }
 }
 RBAN Settings Out {
  type=RBANSettingStruct
  default {
   componentLinearAccScale=<0 0 0>
   componentLinearVelScale=<0 0 0>
   simSpaceSettings {
    dampingAlpha=1
    maxAngularAcceleration=1000
    maxAngularVelocity=3
    maxLinearAcceleration=300
    maxLinearVelocity=700
    velocityScaleZ=0
    worldAlpha=1
    externalAngularVelocity=<0 0 0>
    externalLinearDragV=<0 0.5 0>
    externalLinearVelocity=<0 0 0>
   }
  }
 }
]
